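# Niagara system for the LiverRomaine emote's mailbox effects: floating rings, letter trails and lid-slam impacts.
# Looped-age curves time ring meshes on the mailbox skeleton and tinted ribbon trails on the letters.
# Staggered bursts at the lid bones fire a flash mesh, toon smoke and radiating speedline sprites.
# Spawn groups double as bone/socket indices and select per-group offsets and directions, while time-of-day colors tint trails, flash and smoke.
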NiagaraSystem /BRCosmetics/Animation/Game/MainPlayer/Emotes/LiverRomaine/FX/NS_Emote_LiverRomaine_Mailbox
dependencies {
 usedDataInterfaces [&NiagaraDataInterfaceCurve, &NiagaraDataInterfaceSkeletalMesh, &NiagaraDataInterfaceArrayFloat]
 usedDynamicInputs [&SelectFloatFromArray]
 usedModules [
  &self:SystemSpawnScript.SystemScriptSource.SystemScriptGraph.NiagaraNodeAssignment_0.SetVariables_#1
  &self:SystemSpawnScript.SystemScriptSource.SystemScriptGraph.NiagaraNodeAssignment_1.SetVariables_#2
  &self:SystemSpawnScript.SystemScriptSource.SystemScriptGraph.NiagaraNodeAssignment_2.SetVariables_#3
  &SystemState
  &self:SystemSpawnScript.SystemScriptSource.SystemScriptGraph.NiagaraNodeAssignment_3.SetVariables_#4
  &self:SystemSpawnScript.SystemScriptSource.SystemScriptGraph.NiagaraNodeAssignment_4.SetVariables_#5
 ]
}
emitters {
 Floating_Rings {
  emitterUpdateScript [
   &EmitterState()
   &SpawnRate(
    SpawnRate=&FloatFromCurve(
     FloatCurve=DataInterface() {propertyValues {Curve {keys [<at 0 0>,<at 0.1 1>,<at 2.95 1>,<at 3.05 0>]}}}
     CurveIndex=$System.LoopedAge
     Scale Curve=2
    )
    Use Spawn Probability=false
   )
   &SpawnRate(
    SpawnGroup=1
    SpawnRate=&FloatFromCurve(
     FloatCurve=DataInterface() {propertyValues {Curve {keys [<at 0.1 0>,<at 0.25 1>,<at 3.15 1>,<at 3.2 0>]}}}
     CurveIndex=$System.LoopedAge
     Scale Curve=2
    )
    Use Spawn Probability=false
   ) {module=SpawnRate001}
   &SpawnRate(
    SpawnGroup=2
    SpawnRate=&FloatFromCurve(
     FloatCurve=DataInterface() {propertyValues {Curve {keys [<at 0.7 0>,<at 0.85 1>,<at 3.7 1>,<at 3.75 0>]}}}
     CurveIndex=$System.LoopedAge
     Scale Curve=2
    )
    Use Spawn Probability=false
   ) {module=SpawnRate002}
   &SpawnRate(
    SpawnGroup=3
    SpawnRate=&FloatFromCurve(
     FloatCurve=DataInterface() {propertyValues {Curve {keys [<at 0.9 0>,<at 1 1>,<at 4 1>,<at 4.05 0>]}}}
     CurveIndex=$System.LoopedAge
     Scale Curve=2
    )
    Use Spawn Probability=false
   ) {module=SpawnRate003}
  ]
  particleSpawnScript [
   &InitializeParticle(
    Color=<rgba 1 1 1 0.6>
    Lifetime=1.2
    Mesh Scale=<0.1 0.1 0.05>
    Mesh Scale Mode=::NewEnumerator4
    Position Offset=<30 0 0>
    UsePositionOffset=false
    Write Material Random=true
   )
   &InitialMeshOrientation(
    Mesh Orientation Mode=::NewEnumerator1
    Rotation=&RandomRangeVector(Minimum=<0 0 0>, Maximum=<0 0 1>)
    Use Rotation Vector=true
   )
  ]
  particleUpdateScript [
   &ParticleState()
   &SkeletalMeshLocation(
    Bone / Socket Index=&SpawnGroup()
    Bone Sampling Mode=::NewEnumerator4
    MaskSpawnGroup=false
    Sampled Position Offset=/Game/Effects/Niagara/DynamicInputs/Water/FN_PickVectorFromSpawnGroupCutoff(
     Under Cutoff Vector:Vector3f=&MakeVector(
      X=-1
      Y=2.25
      Z=&FloatFromCurve(FloatCurve=DataInterface() {propertyValues {Curve {keys [<at 0 5>,<at 1 -10>]}}})
     )
     Over Cutoff Vector:Vector3f=&MakeVector(
      X=1
      Y=2.5
      Z=&FloatFromCurve(FloatCurve=DataInterface() {propertyValues {Curve {keys [<at 0 5 RCIM_Cubic>,<at 1 -10 RCIM_Cubic>]}}})
     )
     SpawnGroup:NiagaraInt32=$Emitter.SpawnGroup
     Cutoff:NiagaraInt32=2
    )
    Skeletal Mesh=$System.Mailbox Skel Mesh
   )
   &ScaleMeshSize(Scale Factor=&VectorFromFloat(Value=&FloatFromCurve(FloatCurve=DataInterface() {propertyValues {Curve {
         defaultValue=0
         keys [{interpMode=RCIM_Cubic, tangentMode=RCTM_User, time=0, value=0.4}, {interpMode=RCIM_Cubic, tangentMode=RCTM_User, time=1, value=1}]
        }}})))
   &ScaleColor(
    Scale Alpha=&FloatFromCurve(FloatCurve=DataInterface() {propertyValues {Curve {
        defaultValue=1
        keys [
         {arriveTangent=9.0725, interpMode=RCIM_Cubic, leaveTangent=9.0725, tangentMode=RCTM_User, time=0, value=0}
         {arriveTangent=-2.5249, interpMode=RCIM_Cubic, leaveTangent=-2.5249, tangentMode=RCTM_User, time=0.1408, value=0.9847}
         {arriveTangent=-0.0405, interpMode=RCIM_Cubic, leaveTangent=-0.0405, tangentMode=RCTM_User, time=1, value=0}
        ]
       }}})
    ScaleA=true
    ScaleRGB=true
   )
   &DynamicMaterialParameters(
    Index 0 Param 1=&FloatFromCurve(FloatCurve=DataInterface() {propertyValues {Curve {keys [<at 0 1.5>,<at 0.5 1.5>,<at 1 0>]}}})
    Index 0 Param 2=0.5
    Index 0 Param 4=1
    Module.Index0Float0Write=true
    Module.Index0Float1Write=true
    Module.Index0Float2Write=true
    Module.Index0Float3Write=true
   )
  ]
  rendererClasses [&NiagaraMeshRendererProperties]
  renderers [{rendererClass=&NiagaraMeshRendererProperties, rendererIndex=0}]
 }
 Letter_Trails {
  emitterUpdateScript [
   &EmitterState()
   &SpawnRate(
    SpawnRate=&FloatFromCurve(FloatCurve=DataInterface() {propertyValues {Curve {keys [<at 0 1>,<at 0.15 1>,<at 0.2 0>]}}}, CurveIndex=$System.LoopedAge, Scale Curve=60)
    Use Spawn Probability=false
   )
   &SpawnRate(
    SpawnGroup=1
    SpawnRate=&FloatFromCurve(
     FloatCurve=DataInterface() {propertyValues {Curve {keys [<at 0.25 0>,<at 0.3 1>,<at 0.45 1>,<at 0.5 0>]}}}
     CurveIndex=$System.LoopedAge
     Scale Curve=60
    )
    Use Spawn Probability=false
   ) {module=SpawnRate001}
   &SpawnRate(
    SpawnGroup=2
    SpawnRate=&FloatFromCurve(
     FloatCurve=DataInterface() {propertyValues {Curve {keys [<at 0.7 0>,<at 0.75 1>,<at 1.15 1>,<at 1.25 0>]}}}
     CurveIndex=$System.LoopedAge
     Scale Curve=30
    )
    Use Spawn Probability=false
   ) {module=SpawnRate002}
   &SpawnRate(
    SpawnGroup=3
    SpawnRate=&FloatFromCurve(
     FloatCurve=DataInterface() {propertyValues {Curve {keys [<at 1.4 0>,<at 1.45 1>,<at 1.75 1>,<at 1.8 0>]}}}
     CurveIndex=$System.LoopedAge
     Scale Curve=45
    )
    Use Spawn Probability=false
   ) {module=SpawnRate003}
   &SpawnRate(
    SpawnGroup=4
    SpawnRate=&FloatFromCurve(
     FloatCurve=DataInterface() {propertyValues {Curve {keys [<at 1.4 0>,<at 1.45 1>,<at 1.75 1>,<at 1.8 0>]}}}
     CurveIndex=$System.LoopedAge
     Scale Curve=45
    )
    Use Spawn Probability=false
   ) {module=SpawnRate004}
  ]
  particleSpawnScript [
   &InitializeParticle(
    Color=&MultiplyLinearColor(
     Linear Color=&MaskLinearColorBySpawnGroup(
      Masked Linear Color=<rgba 0.7231 0.8388 0.7758 1>
      Passthrough Linear Color=&MaskLinearColorBySpawnGroup(
       Masked Linear Color=<rgba 0.521 0.1329 0.1095 1>
       Passthrough Linear Color=&MaskLinearColorBySpawnGroup(
        Masked Linear Color=<rgba 0.4851 0.3231 0.1384 1>
        Passthrough Linear Color=&MaskLinearColorBySpawnGroup(Masked Linear Color=<rgba 0.521 0.1329 0.1095 1>, Passthrough Linear Color=<rgba 0.7231 0.8388 0.7758 1>, Spawn Group Mask=3)
        Spawn Group Mask=2
       )
       Spawn Group Mask=1
      )
     )
     Scale Factor=&MakeVector4_FromLinearColor(LinearColor=$System.ToD)
    )
    Lifetime=0.24
    Ribbon Facing Mode=::NewEnumerator1
    Ribbon Facing Vector:Vector3f=$Engine.Owner.SystemXAxis
    Ribbon Width=&MaskFloatBySpawnGroup(
     Masked Float=10
     Passthrough Float=&MaskFloatBySpawnGroup(
      Masked Float=10
      Passthrough Float=&MaskFloatBySpawnGroup(
       Masked Float=10
       Passthrough Float=&MaskFloatBySpawnGroup(Masked Float=8, Passthrough Float=8, Spawn Group Mask=3)
       Spawn Group Mask=2
      )
      Spawn Group Mask=1
     )
    )
    Ribbon Width Mode=::NewEnumerator1
    UsePositionOffset=false
    Write Material Random=true
   )
   &SkeletalMeshLocation(
    Bone / Socket Index=&SpawnGroup()
    Bone Sampling Mode=::NewEnumerator4
    MaskSpawnGroup=false
    Sampled Position Offset=<0 -3 0>
    Skeletal Mesh=$System.Letters Skel Mesh
   )
   &AddVelocity(Velocity=&FN_PickVectorFromSpawnGroupCutoff(
     Under Cutoff Vector:Vector3f=<0 -50 0>
     Over Cutoff Vector:Vector3f=<0 -10 0>
     SpawnGroup:NiagaraInt32=$Emitter.SpawnGroup
     Cutoff:NiagaraInt32=2
    ))
  ]
  particleUpdateScript [
   &ParticleState()
   &ScaleColor(
    Scale Alpha=&FloatFromCurve(FloatCurve=DataInterface() {propertyValues {Curve {
        defaultValue=1
        keys [
         {arriveTangent=9.0725, interpMode=RCIM_Cubic, leaveTangent=9.0725, tangentMode=RCTM_User, time=0, value=0}
         {arriveTangent=-2.5249, interpMode=RCIM_Cubic, leaveTangent=-2.5249, tangentMode=RCTM_User, time=0.1408, value=0.9847}
         {arriveTangent=-0.0405, interpMode=RCIM_Cubic, leaveTangent=-0.0405, tangentMode=RCTM_User, time=1, value=0}
        ]
       }}})
    ScaleA=true
    ScaleRGB=true
   )
   &ScaleRibbonWidth(Ribbon Width Scale=&FloatFromCurve(FloatCurve=DataInterface() {propertyValues {Curve {
        defaultValue=1
        keys [
         {arriveTangent=-2.5249, interpMode=RCIM_Cubic, time=0.1408, value=0.9847}
         {arriveTangent=-1.9535, interpMode=RCIM_Cubic, leaveTangent=-1.9535, tangentMode=RCTM_User, time=1, value=0}
        ]
       }}}))
   &InheritVelocity(Inherited Velocity Amount Scale=&VectorFromFloat(Value=50), Speed Limit=true, UseSpeedThreshold=false) {enabled=false}
   &InheritSourceMovement(
    Applied Position Scale=&VectorFromFloat(Value=0.1)
    ApplyForce=false
    ApplyPosition=true
    ApplyPositionFalloff=false
   ) {enabled=false}
   &SolveForcesAndVelocity(Clamp Velocity=false, Limit Acceleration=false)
  ]
  rendererClasses [&NiagaraRibbonRendererProperties]
  renderers [{rendererClass=&NiagaraRibbonRendererProperties}]
 }
 Mailbox_Shut {
  emitterUpdateScript [
   &EmitterState()
   /SetVariables(Emitter.TexturePOsitionX:NiagaraFloat=&RandomRangeFloat(Minimum=0, Maximum=1)) {isSetParameters=true}
   &SpawnBurst_Instantaneous(Spawn Time=2.6, Use Spawn Probability=false, UseLoopCountLimit=false)
   &SpawnBurst_Instantaneous(Spawn Group=1, Spawn Time=2.45, Use Spawn Probability=false, UseLoopCountLimit=false) {module=SpawnBurst_Instantaneous001}
   &SpawnBurst_Instantaneous(Spawn Group=2, Spawn Time=3.16, Use Spawn Probability=false, UseLoopCountLimit=false) {module=SpawnBurst_Instantaneous002}
   &SpawnBurst_Instantaneous(Spawn Group=3, Spawn Time=3.8, Use Spawn Probability=false, UseLoopCountLimit=false) {module=SpawnBurst_Instantaneous003}
  ]
  particleSpawnScript [&InitializeParticle(
    Color=&MultiplyLinearColor(
     Linear Color=<rgba 5 3.2265 0 1>
     Scale Factor=&MakeVector4_FromLinearColor(LinearColor=&FNTimeOfDayColor(Night Color=<rgba 0.2 0.2 0.2 1>, Morning Color=<rgba 0.6 0.6 0.6 1>, Day Color=<rgba 1 1 1 1>, Evening Color=<rgba 0.6 0.6 0.6 1>))
    )
    Lifetime=0.2
    Mesh Scale=<0.017 0.1 0.017>
    Mesh Scale Mode=::NewEnumerator4
    UsePositionOffset=false
    Write Material Random=true
   )]
  particleUpdateScript [
   &ParticleState()
   &SkeletalMeshLocation(
    Bone / Socket Index=&SpawnGroup()
    Bone Sampling Mode=::NewEnumerator4
    MaskSpawnGroup=false
    Sampled Position Offset=&MaskVectorBySpawnGroup(
     Masked Vector=<0 -2 11>
     Passthrough Vector=&MaskVectorBySpawnGroup(
      Masked Vector=<0 -2 11>
      Passthrough Vector=&MaskVectorBySpawnGroup(Masked Vector=<0.1 -2 11>, Passthrough Vector=<0.1 -2 11>, Spawn Group Mask=2)
      Spawn Group Mask=1
     )
    )
    Skeletal Mesh=$System.Mailbox Lid Skel Mesh
   )
   &ScaleColor(
    Scale Alpha=&FloatFromCurve(FloatCurve=DataInterface() {propertyValues {Curve {
        defaultValue=1
        keys [<at 0 1 RCIM_Cubic>, {arriveTangent=-2.2416, interpMode=RCIM_Cubic, leaveTangent=-2.2416, tangentMode=RCTM_User, time=1, value=0}]
       }}})
    ScaleA=true
    ScaleRGB=true
   )
   &DynamicMaterialParameters(
    Index 0 Param 1=&FloatFromCurve(FloatCurve=DataInterface() {propertyValues {Curve {keys [{arriveTangent=-1.3555, interpMode=RCIM_Cubic, leaveTangent=-1.3555, tangentMode=RCTM_User, time=0, value=0}, <at 1 -1 RCIM_Cubic>]}}})
    Module.Index0Float0Write=true
    Module.Index0Float1Write=true
    Module.Index0Float2Write=true
    Module.Index0Float3Write=true
   )
  ]
  rendererClasses [&NiagaraMeshRendererProperties]
  renderers [{rendererClass=&NiagaraMeshRendererProperties, rendererIndex=0}]
 }
 Mailbox_Shut_Smoke {
  emitterSpawnScript [/Game/Effects/Niagara/MasterMaterials/ToonShader/Template/NM_ToonShaderParameters(BaseColor:LinearColor=<rgba 0.2604 0.2604 0.2604 1>, Lighting_Hardness=0.2)]
  emitterUpdateScript [
   &EmitterState()
   &SpawnBurst_Instantaneous(Spawn Time=2.6, Use Spawn Probability=false, UseLoopCountLimit=false)
   &SpawnBurst_Instantaneous(Spawn Group=1, Spawn Time=2.45, Use Spawn Probability=false, UseLoopCountLimit=false) {module=SpawnBurst_Instantaneous001}
   &SpawnBurst_Instantaneous(Spawn Group=2, Spawn Time=3.15, Use Spawn Probability=false, UseLoopCountLimit=false) {module=SpawnBurst_Instantaneous002}
   &SpawnBurst_Instantaneous(Spawn Group=3, Spawn Time=3.8, Use Spawn Probability=false, UseLoopCountLimit=false) {module=SpawnBurst_Instantaneous003}
  ]
  particleSpawnScript [
   &InitializeParticle(
    Color=$System.ToD
    Lifetime=0.525
    Mesh Scale Mode=::NewEnumerator3
    Mesh Uniform Scale=1.6
    Sprite Rotation Mode=::NewEnumerator2
    UsePositionOffset=false
    Write Material Random=true
   )
   &InitialMeshOrientation(
    Mesh Orientation Mode=::NewEnumerator1
    Rotation=&RandomRangeVector(Minimum=<0 0 0>, Maximum=<0 1 0>)
    Use Rotation Vector=true
   )
  ]
  particleUpdateScript [
   &ParticleState()
   &ScaleMeshSize(Scale Factor=&VectorFromFloat(Value=&FloatFromCurve(FloatCurve=DataInterface() {propertyValues {Curve {keys [<at 0 1 RCIM_Cubic>, {arriveTangent=1.4293, interpMode=RCIM_Cubic, leaveTangent=1.4293, tangentMode=RCTM_User, time=1, value=2.25}]}}})))
   &SkeletalMeshLocation(
    Bone / Socket Index=&SpawnGroup()
    Bone Sampling Mode=::NewEnumerator4
    MaskSpawnGroup=false
    Sampled Position Offset=&MaskVectorBySpawnGroup(
     Masked Vector=<0 0 11>
     Passthrough Vector=&MaskVectorBySpawnGroup(
      Masked Vector=<0 0 11>
      Passthrough Vector=&MaskVectorBySpawnGroup(Masked Vector=<0.1 0 11>, Passthrough Vector=<0.1 0 11>, Spawn Group Mask=2)
      Spawn Group Mask=1
     )
    )
    Skeletal Mesh=$System.Mailbox Lid Skel Mesh
   )
   &ScaleColor(
    Scale Alpha=&FloatFromCurve(FloatCurve=DataInterface() {propertyValues {Curve {
        defaultValue=1
        keys [<at 0 1 RCIM_Cubic>, {arriveTangent=-2.2416, interpMode=RCIM_Cubic, leaveTangent=-2.2416, tangentMode=RCTM_User, time=1, value=0}]
       }}})
    ScaleA=true
    ScaleRGB=true
   )
   &DynamicMaterialParameters(
    Index 0 Param 1=&FloatFromCurve(
     FloatCurve=DataInterface() {propertyValues {Curve {keys [<at 0 0 RCIM_Cubic>, {arriveTangent=2.5195, interpMode=RCIM_Cubic, leaveTangent=2.5195, tangentMode=RCTM_User, time=1, value=1}]}}}
     Scale Curve=0.8
    )
    Module.Index0Float0Write=true
    Module.Index0Float1Write=true
    Module.Index0Float2Write=true
    Module.Index0Float3Write=true
   )
   /SetVariables(Particles.ImpactPos:NiagaraPosition=&AddVectorToPosition(
     Position=$Particles.SkeletalMeshLocation.SampledPosition
     Vector=&FN_PickVectorFromSpawnGroupCutoff(
      Under Cutoff Vector:Vector3f=<-1.5 0 10>
      Over Cutoff Vector:Vector3f=<1.5 0 12>
      SpawnGroup:NiagaraInt32=$Emitter.SpawnGroup
      Cutoff:NiagaraInt32=2
     )
    )) {enabled=false, isSetParameters=true}
  ]
  rendererClasses [&NiagaraMeshRendererProperties]
  renderers [{rendererClass=&NiagaraMeshRendererProperties, rendererIndex=0}]
 }
 Mailbox_Shut_Speedlines {
  emitterUpdateScript [
   &EmitterState()
   &SpawnBurst_Instantaneous(
    Spawn Count=&RandomRangeInt(Minimum=6, Maximum=9)
    Spawn Time=2.6
    Use Spawn Probability=false
    UseLoopCountLimit=false
   )
   &SpawnBurst_Instantaneous(
    Spawn Count=&RandomRangeInt(Minimum=6, Maximum=9)
    Spawn Group=1
    Spawn Time=2.45
    Use Spawn Probability=false
    UseLoopCountLimit=false
   ) {module=SpawnBurst_Instantaneous001}
   &SpawnBurst_Instantaneous(
    Spawn Count=&RandomRangeInt(Minimum=6, Maximum=9)
    Spawn Group=2
    Spawn Time=3.16
    Use Spawn Probability=false
    UseLoopCountLimit=false
   ) {module=SpawnBurst_Instantaneous002}
   &SpawnBurst_Instantaneous(
    Spawn Count=&RandomRangeInt(Minimum=6, Maximum=9)
    Spawn Group=3
    Spawn Time=3.8
    Use Spawn Probability=false
    UseLoopCountLimit=false
   ) {module=SpawnBurst_Instantaneous003}
  ]
  particleSpawnScript [
   &InitializeParticle(
    Color=<rgba 0 0 0 1>
    Lifetime Max=0.3
    Lifetime Min=0.1
    Lifetime Mode=::NewEnumerator1
    Sprite Size Max=<2 40>
    Sprite Size Min=<0.5 15>
    Sprite Size Mode=::NewEnumerator2
    UsePositionOffset=false
    Write Material Random=true
   )
   &SkeletalMeshLocation(
    Bone / Socket Index=&SpawnGroup()
    Bone Sampling Mode=::NewEnumerator4
    MaskSpawnGroup=false
    Sampled Position Offset=&MaskVectorBySpawnGroup(
     Masked Vector=<0 0 11>
     Passthrough Vector=&MaskVectorBySpawnGroup(
      Masked Vector=<0 0 11>
      Passthrough Vector=&MaskVectorBySpawnGroup(Masked Vector=<0 0 11>, Passthrough Vector=<0 0 11>, Spawn Group Mask=2)
      Spawn Group Mask=1
     )
    )
    Skeletal Mesh=$System.Mailbox Lid Skel Mesh
   )
   &ShapeLocation(
    MaskSpawnGroup=false
    Override Seed=false
    Ring / Disc Distribution Mode:ENiagaraSphereDistributionMode=::NewEnumerator1
    Ring Radius=6
    Rotation Mode=::NewEnumerator1
    Shape Primitive=::NewEnumerator4
    U Position=&ReturnNormalizedExecIndex()
    UseOwnerScale=true
   )
   &AddVelocity(
    Cone Angle=90
    Cone Axis=&FN_PickVectorFromSpawnGroupCutoff(
     Under Cutoff Vector:Vector3f=<-1 0 0>
     Over Cutoff Vector:Vector3f=<1 0 0>
     SpawnGroup:NiagaraInt32=$Emitter.SpawnGroup
     Cutoff:NiagaraInt32=2
    )
    Inner Cone Angle=30
    Module.UseConeFalloff=false
    Module.UseVelDistribution=false
    Override Seed=false
    Velocity Mode=::NewEnumerator2
    Velocity Speed=&RandomRangeFloat(Minimum=200, Maximum=325)
   )
  ]
  particleUpdateScript [
   &ParticleState()
   &ScaleSpriteSize(Uniform Curve Sprite Scale=DataInterface() {propertyValues {Curve {keys [{tangentMode=RCTM_User, time=0, value=1}, {tangentMode=RCTM_User, time=1, value=0}]}}})
   &Drag(
    Drag=&FloatFromCurve(
     FloatCurve=DataInterface() {propertyValues {Curve {keys [<at 0 0>,<at 0.1 1>]}}}
     Scale Curve=&RandomRangeFloat(Minimum=0.8, Maximum=1.2)
    )
    Use Linear Drag=true
    Use Rotational Drag=true
   )
   &SolveForcesAndVelocity(Clamp Velocity=false, Limit Acceleration=false)
  ]
  rendererClasses [&NiagaraSpriteRendererProperties]
  renderers [{rendererClass=&NiagaraSpriteRendererProperties}]
 }
}
summary {
 systemName=NS_Emote_LiverRomaine_Mailbox
 emitters [
  {emitterName=Letter_Trails, rendererClasses [&NiagaraRibbonRendererProperties]}
  {emitterName=Floating_Rings, rendererClasses [&NiagaraMeshRendererProperties]}
  {emitterName=Mailbox_Shut_Smoke, rendererClasses [&NiagaraMeshRendererProperties]}
  {emitterName=Mailbox_Shut, rendererClasses [&NiagaraMeshRendererProperties]}
  {emitterName=Mailbox_Shut_Speedlines, rendererClasses [&NiagaraSpriteRendererProperties]}
 ]
 userVariables []
}
