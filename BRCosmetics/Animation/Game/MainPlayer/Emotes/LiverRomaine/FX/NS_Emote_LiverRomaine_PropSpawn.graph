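# Niagara effect for the LiverRomaine emote's prop spawn, producing a short puff of toon-shaded smoke.
# Only Smoke_Toon lists a renderer: it bursts a few randomly oriented meshes from a small sphere with outward velocity, slowed by high drag and offset relative to the camera.
# Smoke color follows time of day, while curves grow mesh scale and drive a dynamic material parameter over each brief lifetime.
# The other five emitters list no renderers.
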NiagaraSystem /BRCosmetics/Animation/Game/MainPlayer/Emotes/LiverRomaine/FX/NS_Emote_LiverRomaine_PropSpawn
dependencies {
 usedDataInterfaces [&NiagaraDataInterfaceCurve, &NiagaraDataInterfaceArrayFloat]
 usedDynamicInputs [&SelectFloatFromArray]
 usedModules [
  &SystemState
  &self:SystemSpawnScript.SystemScriptSource.SystemScriptGraph.NiagaraNodeAssignment_0.SetVariables_#1
 ]
}
emitters {
 AirglowLWE {rendererClasses [], renderers []}
 Smoke_Toon {
  emitterSpawnScript [/Game/Effects/Niagara/MasterMaterials/ToonShader/Template/NM_ToonShaderParameters(HighLight_Softness=0, Lighting_Hardness=0.8, Metallic_LightingPower=0, RimLight_Color:LinearColor=<rgba 0 0 0 1>, RimLight_Width=0)]
  emitterUpdateScript [
   &EmitterState()
   &SpawnBurst_Instantaneous(Spawn Count=5, Use Spawn Probability=false, UseLoopCountLimit=false)
  ]
  particleSpawnScript [
   &InitializeParticle(
    Color=&FNTimeOfDayColor(
     Night Color=<rgba 0.1039 0.1554 0.3837 1>
     Morning Color=<rgba 0.2448 0.1986 0.1928 1>
     Day Color=<rgba 0.2448 0.2448 0.2448 1>
     Evening Color=<rgba 0.1921 0.1606 0.2124 1>
    )
    Lifetime=&RandomRangeFloat(Minimum=0.3, Maximum=0.6)
    Mesh Scale Mode=::NewEnumerator1
    Mesh Uniform Scale Max=1.6
    Mesh Uniform Scale Min=1.2
    Sprite Size Mode=::NewEnumerator3
    Uniform Sprite Size=3.1789
    UsePositionOffset=false
    Write Material Random=true
   )
   &InitialMeshOrientation(Rotation=&RandomRangeVector(Minimum=<0 0 0>, Maximum=<1 1 1>), Use Rotation Vector=true)
   &ShapeLocation(
    Hemisphere Distribution=<0.5 1>
    MaskSpawnGroup=false
    Override Seed=false
    Rotation Angle=0.75
    Rotation Axis=<0 0 1>
    Rotation Mode=::NewEnumerator1
    Sphere Radius=7
    Sphere Surface Distribution=0.5
    UseOwnerScale=true
   )
   &AddVelocity(
    Module.Use Falloff Distance=false
    Override Seed=false
    Velocity Mode=::NewEnumerator1
    Velocity Origin=$StackContext.Position
    Velocity Speed=&RandomRangeFloat(Minimum=100, Maximum=150)
   )
   &CameraOffset(Camera Offset Amount=&RandomRangeFloat(Minimum=10, Maximum=15))
  ]
  particleUpdateScript [
   &ParticleState()
   &ScaleMeshSize(Scale Factor=&VectorFromFloat(Value=&FloatFromCurve(FloatCurve=DataInterface() {propertyValues {Curve {keys [
          {arriveTangent=5.3301, interpMode=RCIM_Cubic, leaveTangent=5.3301, tangentMode=RCTM_User, time=0, value=0}
          {interpMode=RCIM_Cubic, tangentMode=RCTM_User, time=0.7, value=0.95}
          {tangentMode=RCTM_User, time=1, value=1}
         ]}}})))
   &Drag(Drag=&RandomRangeFloat(Minimum=9, Maximum=13), Use Linear Drag=true, Use Rotational Drag=false)
   &SolveForcesAndVelocity(Clamp Velocity=false, Limit Acceleration=false)
   &DynamicMaterialParameters(
    Index 0 Param 1=&FloatFromCurve(FloatCurve=DataInterface() {propertyValues {Curve {keys [
         {arriveTangent=2.2672, interpMode=RCIM_Cubic, leaveTangent=2.2672, tangentMode=RCTM_User, time=0, value=0}
         {interpMode=RCIM_Cubic, tangentMode=RCTM_User, time=1, value=1}
        ]}}})
    Module.Param0WriteEnabled=true
    Module.Param1WriteEnabled=true
    Module.Param2WriteEnabled=true
    Module.Param3WriteEnabled=true
   )
  ]
  rendererClasses [&NiagaraMeshRendererProperties]
  renderers [{rendererClass=&NiagaraMeshRendererProperties, rendererIndex=0}]
 }
 SparklesLWE {rendererClasses [], renderers []}
 SparklesLWE001 {rendererClasses [], renderers []}
 ToonImpactSmoke {rendererClasses [], renderers []}
 ToonSmoke_LowMedium {rendererClasses [], renderers []}
}
summary {
 systemName=NS_Emote_LiverRomaine_PropSpawn
 emitters [
  {emitterName=AirglowLWE, rendererClasses []}
  {emitterName=SparklesLWE, rendererClasses []}
  {emitterName=ToonImpactSmoke, rendererClasses []}
  {emitterName=ToonSmoke_LowMedium, rendererClasses []}
  {emitterName=SparklesLWE001, rendererClasses []}
  {emitterName=Smoke_Toon, rendererClasses [&NiagaraMeshRendererProperties]}
 ]
 userVariables [User.NormalizedNotifyProgress {type=NiagaraFloat, defaultValue=0, description=""}]
}
